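# Physics asset for the Pencil VestJacket on the medium female player body, setting up capsule bodies for the jacket's dynamic bones.
# Seven chains (back, back corners, sides, fronts) of three or four links hang from MSTR bodies, with per-chain swing limits and zero twist limits.
# Pelvis, lower spine and thigh capsules provide body shapes; tiny Aim bodies have no constraints.
# The lower right-front links carry doubled capsules for wider coverage.
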PhysicsAsset /BRCosmetics/Characters/Player/Female/Medium/Bodies/F_MED_Pencil_B/VestJacket/Meshes/Pencil_VestJacket_MedB_Ubdy_PhysicsAsset
bodies {
 dyn_jacket_MSTR [{shapeType=Capsule, length=0.505, radius=0.505}]
 dyn_jacket_bk_01 [{shapeType=Capsule, length=10.875, radius=6.625, shapeName=dyn_jacket_bk_01_capsule, center=<4 -5 0>}]
 dyn_jacket_bk_02 [{shapeType=Capsule, length=10.875, radius=6.625, shapeName=dyn_jacket_bk_02_capsule, center=<4 -5 0>}]
 dyn_jacket_bk_03 [{shapeType=Capsule, length=10.875, radius=6.625, shapeName=dyn_jacket_bk_03_capsule, center=<4 -5 0>}]
 dyn_jacket_bk_Aim [{shapeType=Capsule, length=0.505, radius=0.505, shapeName=dyn_jacket_bk_Aim_capsule, rotation=<rot 0 -90 0>}]
 dyn_jacket_bk_MSTR [{shapeType=Capsule, length=0.505, radius=0.505, shapeName=dyn_jacket_bk_MSTR_capsule, rotation=<rot 0 -90 0>}]
 dyn_jacket_bk_l_01 [{shapeType=Capsule, length=10, radius=6.625, shapeName=dyn_jacket_bk_l_01_capsule, center=<4 0 5>, rotation=<rot 0 0 90>}]
 dyn_jacket_bk_l_02 [{shapeType=Capsule, length=10, radius=6.625, shapeName=dyn_jacket_bk_l_02_capsule, center=<4 0 5>, rotation=<rot 0 0 90>}]
 dyn_jacket_bk_l_03 [{shapeType=Capsule, length=10, radius=6.625, shapeName=dyn_jacket_bk_l_03_capsule, center=<4 0 5>, rotation=<rot 0 0 90>}]
 dyn_jacket_bk_l_Aim [{shapeType=Capsule, length=0.505, radius=0.505, shapeName=dyn_jacket_bk_l_Aim_capsule, rotation=<rot 0 -90 0>}]
 dyn_jacket_bk_l_MSTR [{shapeType=Capsule, length=0.505, radius=0.505, shapeName=dyn_jacket_bk_l_MSTR_capsule, rotation=<rot 0 -90 0>}]
 dyn_jacket_bk_r_01 [{shapeType=Capsule, length=10, radius=6.625, shapeName=dyn_jacket_bk_r_01_capsule, center=<4 -5 0>}]
 dyn_jacket_bk_r_02 [{shapeType=Capsule, length=10, radius=6.625, shapeName=dyn_jacket_bk_r_02_capsule, center=<4 -5 0>}]
 dyn_jacket_bk_r_03 [{shapeType=Capsule, length=10, radius=6.625, shapeName=dyn_jacket_bk_r_03_capsule, center=<4 -5 0>}]
 dyn_jacket_bk_r_Aim [{shapeType=Capsule, length=0.505, radius=0.505, shapeName=dyn_jacket_bk_r_Aim_capsule, rotation=<rot 0 -90 0>}]
 dyn_jacket_bk_r_MSTR [{shapeType=Capsule, length=0.505, radius=0.505, shapeName=dyn_jacket_bk_r_MSTR_capsule, rotation=<rot 0 -90 0>}]
 dyn_jacket_fr_l_01 [{shapeType=Capsule, length=5.375, radius=6.625, shapeName=dyn_jacket_fr_l_01_capsule, center=<4 -6 5>, rotation=<rot 0.0001 0.0001 80>}]
 dyn_jacket_fr_l_02 [{shapeType=Capsule, length=8.25, radius=5.75, shapeName=dyn_jacket_fr_l_02_capsule, center=<4 -6 5>, rotation=<rot 0.0001 0.0001 80>}]
 dyn_jacket_fr_l_03 [{shapeType=Capsule, length=15.125, radius=7.75, shapeName=dyn_jacket_fr_l_03_capsule, center=<4 -8 5>, rotation=<rot 0.0001 0 80>}]
 dyn_jacket_fr_l_04 [{shapeType=Capsule, length=11.125, radius=7.75, shapeName=dyn_jacket_fr_l_04_capsule, center=<4 -8 5>, rotation=<rot 0.0001 0 80>}]
 dyn_jacket_fr_l_Aim [{shapeType=Capsule, length=0.505, radius=0.505, shapeName=dyn_jacket_fr_l_Aim_capsule, rotation=<rot 0 -90 0>}]
 dyn_jacket_fr_l_MSTR [{shapeType=Capsule, length=0.505, radius=0.505, shapeName=dyn_jacket_fr_l_MSTR_capsule, rotation=<rot 0 -90 0>}]
 dyn_jacket_fr_r_01 [{shapeType=Capsule, length=5.375, radius=6.625, shapeName=dyn_jacket_fr_r_01_capsule, center=<4 7 4>, rotation=<rot 0.0001 -0.0001 115>}]
 dyn_jacket_fr_r_02 [{shapeType=Capsule, length=6.75, radius=5.25, shapeName=dyn_jacket_fr_r_02_capsule, center=<4 6 5>, rotation=<rot 0 0 90>}]
 dyn_jacket_fr_r_03 [
  {shapeType=Capsule, length=7.625, radius=6.375, shapeName=dyn_jacket_fr_r_03_capsule, center=<4 6 5>, rotation=<rot 0.0001 -0.0001 95>}
  {
   shapeType=Capsule
   length=7.625
   radius=6.375
   shapeName=dyn_jacket_fr_r_03_capsule
   center=<4 16.7565 1.6867>
   rotation=<rot 0.0002 -0.0002 120.0001>
  }
 ]
 dyn_jacket_fr_r_04 [
  {shapeType=Capsule, length=7.625, radius=6.375, shapeName=dyn_jacket_fr_r_04_capsule, center=<4 6 5>, rotation=<rot 0.0001 -0.0001 95>}
  {
   shapeType=Capsule
   length=7.625
   radius=6.375
   shapeName=dyn_jacket_fr_r_04_capsule
   center=<4 16.7565 1.6867>
   rotation=<rot 0.0002 -0.0002 120.0001>
  }
 ]
 dyn_jacket_fr_r_Aim [{shapeType=Capsule, length=0.505, radius=0.505, shapeName=dyn_jacket_fr_r_Aim_capsule, rotation=<rot 0 -90 0>}]
 dyn_jacket_fr_r_MSTR [{shapeType=Capsule, length=0.505, radius=0.505, shapeName=dyn_jacket_fr_r_MSTR_capsule, rotation=<rot 0 -90 0>}]
 dyn_jacket_l_01 [{shapeType=Capsule, length=5.375, radius=6.625, shapeName=dyn_jacket_l_01_capsule, center=<4 0 5>, rotation=<rot 0 0 90>}]
 dyn_jacket_l_02 [{shapeType=Capsule, length=12.75, radius=6.625, shapeName=dyn_jacket_l_02_capsule, center=<4 -4 5>, rotation=<rot 0 -9.9999 90>}]
 dyn_jacket_l_03 [{shapeType=Capsule, length=11.125, radius=6.625, shapeName=dyn_jacket_l_03_capsule, center=<4 -4 5>, rotation=<rot 0 -24.9998 89.9999>}]
 dyn_jacket_l_Aim [{shapeType=Capsule, length=0.505, radius=0.505, shapeName=dyn_jacket_l_Aim_capsule, rotation=<rot 0 -90 0>}]
 dyn_jacket_l_MSTR [{shapeType=Capsule, length=0.505, radius=0.505, shapeName=dyn_jacket_l_MSTR_capsule, rotation=<rot 0 -90 0>}]
 dyn_jacket_r_01 [{shapeType=Capsule, length=5.375, radius=6.625, shapeName=dyn_jacket_r_01_capsule, center=<5 -5 0>}]
 dyn_jacket_r_02 [{shapeType=Capsule, length=5.375, radius=6.625, shapeName=dyn_jacket_r_02_capsule, center=<5 -5 0>, rotation=<rot 9.9999 0 0>}]
 dyn_jacket_r_03 [{shapeType=Capsule, length=5.375, radius=6.625, shapeName=dyn_jacket_r_03_capsule, center=<5 -5 0>, rotation=<rot 24.9998 0 0.0001>}]
 dyn_jacket_r_Aim [{shapeType=Capsule, length=0.505, radius=0.505, shapeName=dyn_jacket_r_Aim_capsule, rotation=<rot 0 -90 0>}]
 dyn_jacket_r_MSTR [{shapeType=Capsule, length=0.505, radius=0.505, shapeName=dyn_jacket_r_MSTR_capsule, rotation=<rot 0 -90 0>}]
 pelvis [
  {shapeType=Capsule, length=9.5875, radius=9.3051, center=<-0.6285 -1.8567 0>}
  {shapeType=Capsule, length=10.665, radius=8.8452, shapeName=pelvis_capsule, center=<-3.9978 -3.5897 0>}
 ]
 spine_01 [{shapeType=Capsule, length=5.1863, radius=9.4291, center=<0 -0.1396 -0.0001>}]
 spine_02 [{shapeType=Capsule, length=3.5955, radius=9.005, center=<0 0.6676 -0.0001>}]
 thigh_l [{shapeType=Capsule, length=30.6385, radius=10.6275, center=<-21.4522 -1.1707 -0.5244>, rotation=<rot 85 179.8042 179.9999>}]
 thigh_r [{shapeType=Capsule, length=30.0092, radius=11.2569, center=<20.9715 1.5 -0.2502>, rotation=<rot 85 179.9997 179.9999>}]
}
constraints [
 {bone1Name=dyn_jacket_bk_01, bone2Name=dyn_jacket_bk_MSTR, swing1LimitDegrees=45, swing2LimitDegrees=45, twistLimitDegrees=0}
 {bone1Name=dyn_jacket_bk_02, bone2Name=dyn_jacket_bk_01, swing1LimitDegrees=45, swing2LimitDegrees=45, twistLimitDegrees=0}
 {bone1Name=dyn_jacket_bk_03, bone2Name=dyn_jacket_bk_02, swing1LimitDegrees=45, swing2LimitDegrees=45, twistLimitDegrees=0}
 {bone1Name=dyn_jacket_bk_l_01, bone2Name=dyn_jacket_bk_l_MSTR, swing1LimitDegrees=35, swing2LimitDegrees=45, twistLimitDegrees=0}
 {bone1Name=dyn_jacket_bk_l_02, bone2Name=dyn_jacket_bk_l_01, swing1LimitDegrees=35, swing2LimitDegrees=45, twistLimitDegrees=0}
 {bone1Name=dyn_jacket_bk_l_03, bone2Name=dyn_jacket_bk_l_02, swing1LimitDegrees=35, swing2LimitDegrees=45, twistLimitDegrees=0}
 {bone1Name=dyn_jacket_bk_r_01, bone2Name=dyn_jacket_bk_r_MSTR, swing1LimitDegrees=45, swing2LimitDegrees=35, twistLimitDegrees=0}
 {bone1Name=dyn_jacket_bk_r_02, bone2Name=dyn_jacket_bk_r_01, swing1LimitDegrees=45, swing2LimitDegrees=35, twistLimitDegrees=0}
 {bone1Name=dyn_jacket_bk_r_03, bone2Name=dyn_jacket_bk_r_02, swing1LimitDegrees=45, swing2LimitDegrees=35, twistLimitDegrees=0}
 {bone1Name=dyn_jacket_fr_l_01, bone2Name=dyn_jacket_fr_l_MSTR, swing1LimitDegrees=15, swing2LimitDegrees=15, twistLimitDegrees=0}
 {bone1Name=dyn_jacket_fr_l_02, bone2Name=dyn_jacket_fr_l_01, swing1LimitDegrees=20, swing2LimitDegrees=30, twistLimitDegrees=0}
 {bone1Name=dyn_jacket_fr_l_03, bone2Name=dyn_jacket_fr_l_02, swing1LimitDegrees=20, swing2LimitDegrees=30, twistLimitDegrees=0}
 {bone1Name=dyn_jacket_fr_l_04, bone2Name=dyn_jacket_fr_l_03, swing1LimitDegrees=20, swing2LimitDegrees=30, twistLimitDegrees=0}
 {bone1Name=dyn_jacket_fr_r_01, bone2Name=dyn_jacket_fr_r_MSTR, swing1LimitDegrees=15, swing2LimitDegrees=15, twistLimitDegrees=0}
 {bone1Name=dyn_jacket_fr_r_02, bone2Name=dyn_jacket_fr_r_01, swing1LimitDegrees=25, swing2LimitDegrees=35, twistLimitDegrees=0}
 {bone1Name=dyn_jacket_fr_r_03, bone2Name=dyn_jacket_fr_r_02, swing1LimitDegrees=25, swing2LimitDegrees=35, twistLimitDegrees=0}
 {bone1Name=dyn_jacket_fr_r_04, bone2Name=dyn_jacket_fr_r_03, swing1LimitDegrees=25, swing2LimitDegrees=35, twistLimitDegrees=0}
 {bone1Name=dyn_jacket_l_01, bone2Name=dyn_jacket_l_MSTR, swing1LimitDegrees=22, swing2LimitDegrees=45, twistLimitDegrees=0}
 {bone1Name=dyn_jacket_l_02, bone2Name=dyn_jacket_l_01, swing1LimitDegrees=22, swing2LimitDegrees=45, twistLimitDegrees=0}
 {bone1Name=dyn_jacket_l_03, bone2Name=dyn_jacket_l_02, swing1LimitDegrees=22, swing2LimitDegrees=45, twistLimitDegrees=0}
 {bone1Name=dyn_jacket_r_01, bone2Name=dyn_jacket_r_MSTR, swing1LimitDegrees=45, swing2LimitDegrees=22, twistLimitDegrees=0}
 {bone1Name=dyn_jacket_r_02, bone2Name=dyn_jacket_r_01, swing1LimitDegrees=45, swing2LimitDegrees=22, twistLimitDegrees=0}
 {bone1Name=dyn_jacket_r_03, bone2Name=dyn_jacket_r_02, swing1LimitDegrees=45, swing2LimitDegrees=22, twistLimitDegrees=0}
]
